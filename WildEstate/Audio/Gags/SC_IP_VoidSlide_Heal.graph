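# Sound cue for the VoidSlide heal sound, reusing four Slurp Barrel heal waves.
# An equal-weight Random node picks one wave and feeds a default Modulator at the root.
# It plays at reduced volume in the consumables sound class with consumable-use attenuation.
# A concurrency override limits copies per owner, stops the oldest, spaces retriggers and scales down the volume of overlapping sounds.
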SoundCue /WildEstate/Audio/Gags/SC_IP_VoidSlide_Heal
root=Modulator_0
nodes [
 {id=Modulator_0, children [Random_0]}
 {id=Random_0, children [WavePlayer_0,WavePlayer_1,WavePlayer_2,WavePlayer_3], properties {weights [1,1,1,1]}}
 {id=WavePlayer_0, properties {soundWaveAssetPtr=&/Item_Slurp_Barrel_Heal_04}}
 {id=WavePlayer_1, properties {soundWaveAssetPtr=&/Item_Slurp_Barrel_Heal_03}}
 {id=WavePlayer_2, properties {soundWaveAssetPtr=&/Item_Slurp_Barrel_Heal_02}}
 {id=WavePlayer_3, properties {soundWaveAssetPtr=&/Item_Slurp_Barrel_Heal_01}}
]
properties {
 attenuationSettings=&Fort_Consumable_Use_Athena_Att
 bOverrideConcurrency=true
 maxDistance=2250
 soundClassObject=&Fort_Consumables
 volumeMultiplier=0.5
 concurrencyOverrides {bLimitToOwner=true, maxCount=2, resolutionRule=StopOldest, retriggerTime=0.5, volumeScale=0.25, volumeScaleAttackTime=1}
}
