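# Sound cue for the Void Slide music stinger, filed among the WildEstate audio gags.
# A single wave player feeds the root directly with the S38 Void Slide stinger, with no randomization, mixing or looping node.
# It routes through the POI music sound class and carries a very large max distance plus an asset user data object.
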SoundCue /WildEstate/Audio/Gags/SC_IP_VoidSlide_Music
root=WavePlayer_0
nodes [{id=WavePlayer_0, properties {soundWaveAssetPtr=&/MX_S38_IP_VoidSlide_STNG}}]
properties {
 maxDistance=2097152
 soundClassObject=/Game/Sounds/FortSoundClassesAndModes/Fort_Music_POI_PSM
 assetUserData [&self:AudioAssetUserData_0]
}
