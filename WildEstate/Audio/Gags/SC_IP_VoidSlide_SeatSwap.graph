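# Sound cue for the VoidSlide SeatSwap sound in the Gags audio set, using emote attenuation and the emote music sound class.
# A Random node picks one of three equally weighted SeatSwap wave variations.
# A Modulator at the root applies slight pitch and volume variation to the chosen wave.
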SoundCue /WildEstate/Audio/Gags/SC_IP_VoidSlide_SeatSwap
root=Modulator_0
nodes [
 {id=Modulator_0, children [Random_0], properties {pitchMax=1.02, pitchMin=0.98, volumeMax=1.02, volumeMin=0.98}}
 {id=Random_0, children [WavePlayer_3,WavePlayer_4,WavePlayer_5], properties {weights [1,1,1]}}
 {id=WavePlayer_3, properties {soundWaveAssetPtr=&/IP_VoidSlide_SeatSwap_01}}
 {id=WavePlayer_4, properties {soundWaveAssetPtr=&/IP_VoidSlide_SeatSwap_02}}
 {id=WavePlayer_5, properties {soundWaveAssetPtr=&/IP_VoidSlide_SeatSwap_03}}
]
properties {
 attenuationSettings=&Fort_Emotes_Athena_Att
 maxDistance=1600
 soundClassObject=&Fort_Music_Emote
 volumeMultiplier=1
}
