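# Sound cue for the sit moment of the VoidSlide gag in WildEstate, judging by its name and folder.
# A single wave player feeds a modulator that adds slight random pitch and volume variation on each play.
# It uses the emote music sound class and Athena emote attenuation, so it plays as a positioned, distance-limited sound.
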SoundCue /WildEstate/Audio/Gags/SC_IP_VoidSlide_Sit
root=Modulator_0
nodes [
 {id=Modulator_0, children [WavePlayer_0], properties {pitchMax=1.02, pitchMin=0.98, volumeMax=1.02, volumeMin=0.98}}
 {id=WavePlayer_0, properties {soundWaveAssetPtr=&/IP_VoidSlide_Sit_01}}
]
properties {
 attenuationSettings=&Fort_Emotes_Athena_Att
 maxDistance=1600
 soundClassObject=&Fort_Music_Emote
 volumeMultiplier=1
}
